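; Asset Data Utility Tool editor action that adds an array of asset data to a named Content Browser collection.
; It runs only when the array has entries and the collection name is set, using the Collection Manager scripting subsystem.
; An existing collection is optionally emptied, and a missing one is optionally created with the configured share type.
; It then adds the assets to that collection in the "Game" container and returns the result.
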(fn RunActionOnAssetDataArray
  (params (AssetDataArray :type Array%20of%20Asset%20Data%20Structures))
  (LocalVariables|Setl_assets AssetDataArray)
  (if (Math|Boolean|ANDBoolean (> (Utilities|Array|Length (LocalVariables|Getl_assets)) 0) (!= (Variables|Default|GetCollectionName) "") (!= (Variables|Default|GetCollectionName) ""))
    (bind _returnvalue (CollectionManager|CollectionExists :self (EditorSubsystems|GetCollectionManagerScriptingSubsystem) :Collection (Variables|Default|GetCollectionName) :ShareType (Variables|Default|GetNewCollectionType)))
    (if _returnvalue
      (if (Variables|Default|GetEmptyCollectionIfExists?)
        (CollectionManager|EmptyCollection (EditorSubsystems|GetCollectionManagerScriptingSubsystem) (Utilities|Struct|MakeCollection "Game" (Variables|Default|GetCollectionName) (Variables|Default|GetNewCollectionType))))
      (elif (Variables|Default|GetCreateNewIfDoesntExist?)
        (CollectionManager|CreateCollection :self (EditorSubsystems|GetCollectionManagerScriptingSubsystem) :Collection (Variables|Default|GetCollectionName) :ShareType (Variables|Default|GetNewCollectionType))))
    (bind _returnvalue_2 (CollectionManager|AddAssetDatastoCollection (EditorSubsystems|GetCollectionManagerScriptingSubsystem) (Utilities|Struct|MakeCollection "Game" (Variables|Default|GetCollectionName) (Variables|Default|GetNewCollectionType)) (LocalVariables|Getl_assets)))
    (return _returnvalue_2)))
